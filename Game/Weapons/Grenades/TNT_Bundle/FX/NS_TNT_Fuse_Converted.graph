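# Niagara system for the burning fuse of the TNT bundle grenade, built from three looping emitters.
# FireGradient splits particles by partition into glow and flame sprites, plus a light with waveform-driven radius.
# GPU LongLived_drips uses spawn groups to mix falling, dragged sparks with short-lived flickers, inheriting source movement.
# Ribbon_Smoke spawns per distance moved, leaving a rising, widening trail that shifts from orange to dark smoke while fading.
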NiagaraSystem /Game/Weapons/Grenades/TNT_Bundle/FX/NS_TNT_Fuse_Converted
dependencies {
 usedDataInterfaces [
  &NiagaraDataInterfaceColorCurve
  &NiagaraDataInterfaceCurve
  &NiagaraDataInterfaceArrayFloat3
  &NiagaraDataInterfaceArrayFloat
  &NiagaraDataInterfaceVectorCurve
  &NiagaraDataInterfaceVector2DCurve
 ]
 usedDynamicInputs []
 usedModules [&SystemState]
}
emitters {
 FireGradient {
  emitterUpdateScript [
   &EmitterState(Inactive Response=::NewEnumerator1, Life Cycle Mode=::NewEnumerator1, Loop Delay=0.27, Loop Duration=1, UseLoopDelay=false)
   &SpawnRate(SpawnRate=50, Use Spawn Probability=false)
   /SetVariables(
    Emitter.LightColor:LinearColor=&MultiplyLinearColor(
     Linear Color=&ColorFromCurve(
      ColorCurve=DataInterface() {propertyValues {
        AlphaCurve {keys [<at 0 1 RCIM_Cubic>]}
        BlueCurve {keys [<at 0 0 RCIM_Cubic>]}
        GreenCurve {keys [<at 0 3.6563 RCIM_Cubic>]}
        RedCurve {keys [<at 0 12 RCIM_Cubic>]}
       }}
      CurveIndex=$Emitter.NormalizedLoopAge
     )
     Scale Factor=&MakeVector4_FromFloat(Float=&FloatFromCurve(
       FloatCurve=DataInterface() {propertyValues {Curve {keys [{tangentMode=RCTM_User, time=0.5, value=0}, {tangentMode=RCTM_User, time=1, value=1}]}}}
       CurveIndex=$Emitter.Age
      ))
    )
    Emitter.LightExponent=2.5
    Emitter.LightRadius:NiagaraFloat=&Waveform(
     Waveform Count=::NewEnumerator1
     Global Amplitude Scale=100
     "Waveform [1]"=::NewEnumerator0
     "[1] Phase Scale"=6
     "[1] Frequency"=1
     "[1] Amplitude Min/Max"=<0 1>
     "[1] Bias"=0.5
     "[2] Average"=false
     "[2] Input Position"=0
     "[2] Phase"=$Emitter.Age
     "[2] Phase Scale"=1
     "[2] Compound Frequency"=2
     "[2] Frequency"=2
     "[2] Amplitude Min/Max"=<0 1>
     "[2] Scale"=1
     "[2] Bias"=0.5
     "[2] Exponent"=1
     "[2] Pulse Duration"=0.1
     "[2] Clip Triangle Min/Max"=<0 0>
     "[2] Random Seed"=0
    )
    Emitter.Position:NiagaraPosition=<0 0 12>
   ) {isSetParameters=true}
  ]
  particleSpawnScript [
   &PartitionParticles()
   /SetVariables(Particles.VisibilityTag=$Particles.PartitionParticles.Partition) {isSetParameters=true}
   &InitializeParticle(
    Lifetime=&MakeCustomFloatFromBool(
     Bool=&SetBoolByFloatComparison(A=0.5, B=&MakeFloatFromInt(Int=$Particles.PartitionParticles.Partition))
     True Float=&RandomRangeFloat(Minimum=0.1, Maximum=0.2)
     False Float=&RandomRangeFloat(Minimum=0.025, Maximum=0.1)
    )
    Position Offset=<0 0 12>
    Sprite Rotation Mode=::NewEnumerator2
    Sprite Size Mode=::NewEnumerator3
    Uniform Sprite Size=&MakeCustomFloatFromBool(
     Bool=&SetBoolByFloatComparison(A=0.5, B=&MakeFloatFromInt(Int=$Particles.PartitionParticles.Partition))
     True Float=75
     False Float=&RandomRangeFloat(Minimum=10, Maximum=25)
    )
    UsePositionOffset=false
    Write Material Random=true
   )
   &Color(Color=&MakeLinearColorFromVectorAndFloat(
     "Vector (RGB)"=&SelectVectorFromArray(
      Direct Array Index=$Particles.PartitionParticles.Partition
      Vector Selection Array=DataInterface() {propertyValues {FloatData [<5 5 5>, <5 1.3858 0.7867>]}}
     )
     "Float (Alpha)"=&SelectFloatFromArray(Direct Array Index=$Particles.PartitionParticles.Partition, Float Selection Array=DataInterface() {propertyValues {FloatData [0.1,1]}})
    ))
  ]
  particleUpdateScript [
   &ParticleState()
   &ScaleSpriteSize(
    Scale Sprite Size Mode=::NewEnumerator0
    Uniform Scale Factor=&MakeCustomFloatFromBool(
     Bool=&SetBoolByFloatComparison(A=0.5, B=&MakeFloatFromInt(Int=$Particles.PartitionParticles.Partition))
     True Float=1
     False Float=&FloatFromCurve(FloatCurve=DataInterface() {propertyValues {Curve {keys [{tangentMode=RCTM_User, time=0, value=0}, {tangentMode=RCTM_User, time=1, value=1}]}}})
    )
   )
   &ScaleColor(
    Scale Alpha=&FloatFromCurve(FloatCurve=DataInterface() {propertyValues {Curve {keys [
         {interpMode=RCIM_Cubic, tangentMode=RCTM_User, time=0, value=0}
         {interpMode=RCIM_Cubic, tangentMode=RCTM_User, time=0.5, value=1}
         {tangentMode=RCTM_User, time=1, value=0}
        ]}}})
    Scale RGB=&MakeCustomVectorFromBool(
     Bool=&SetBoolByFloatComparison(A=0.5, B=&MakeFloatFromInt(Int=$Particles.PartitionParticles.Partition))
     True Vector=&VectorFromCurve(
      VectorCurve=DataInterface() {propertyValues {
        XCurve {keys [<at 0 0.11>,<at 0.5 1>,<at 1 0.11>]}
        YCurve {keys [<at 0 0.05>,<at 0.5 0.13>,<at 1 0.05>]}
        ZCurve {keys [
          {arriveTangent=-0.3559, interpMode=RCIM_Cubic, leaveTangent=-0.3559, tangentMode=RCTM_User, time=0, value=0.03}
          {arriveTangent=0.6186, interpMode=RCIM_Cubic, leaveTangent=-0.8742, tangentMode=RCTM_Break, time=0.5, value=0}
          <at 1 0.03>
         ]}
       }}
      Scale Curve=<2 2 2>
     )
     False Vector=<1 1 1>
    )
    ScaleA=true
    ScaleRGB=true
   )
   &DynamicMaterialParameters(
    Index 0 Param 1=&FloatFromCurve(FloatCurve=DataInterface() {propertyValues {Curve {keys [<at 0 8>,<at 0.5 5>,<at 1 8>]}}})
    Index 0 Param 2=1
    Index 0 Param 3=1
    Index 0 Param 4=1
    Module.Index0Param0Write=true
    Module.Index0Param1Write=true
    Module.Index0Param2Write=true
    Module.Index0Param3Write=true
   )
   &CameraOffset(Camera Offset Amount=&SelectFloatFromArray(Direct Array Index=$Particles.PartitionParticles.Partition, Float Selection Array=DataInterface() {propertyValues {FloatData [50,2]}}))
   /Niagara/Modules/Update/SubUV/SubUVAnimation(Number Of Frames=&SelectFloatFromArray(Direct Array Index=$Particles.PartitionParticles.Partition, Float Selection Array=DataInterface() {propertyValues {FloatData [0,3]}}))
  ]
  rendererClasses [&NiagaraSpriteRendererProperties, &NiagaraLightRendererProperties]
  renderers [
   {rendererClass=&NiagaraSpriteRendererProperties}
   {rendererClass=&NiagaraLightRendererProperties, rendererIndex=1}
   {rendererClass=&NiagaraSpriteRendererProperties, rendererIndex=2}
  ]
 }
 LongLived_drips {
  simTarget=GPUComputeSim
  emitterUpdateScript [
   &EmitterState(Life Cycle Mode=::NewEnumerator1, Loop Delay=0.2, Loop Duration=1, UseLoopDelay=false)
   &SpawnRate(
    SpawnRate=&FloatFromCurve(
     FloatCurve=DataInterface() {propertyValues {Curve {keys [<at 0.75 75 RCIM_Cubic>, {arriveTangent=-14.2993, leaveTangent=-14.2993, time=1, value=0}]}}}
     CurveIndex=$Emitter.NormalizedLoopAge
    )
    Use Spawn Probability=false
   )
   &SpawnRate(SpawnGroup=1, SpawnRate=50, Use Spawn Probability=false) {module=SpawnRate001}
  ]
  particleSpawnScript [
   /SetVariables(Particles.VisibilityTag=$Emitter.SpawnGroup) {isSetParameters=true}
   &InitializeParticle(
    Lifetime=&MaskFloatBySpawnGroup(
     Masked Float=&RandomRangeFloat(Minimum=0.1, Maximum=0.5)
     Passthrough Float=&RandomRangeFloat(Minimum=0.025, Maximum=0.1)
    )
    Position Offset=<0 0 50>
    Sprite Size=&MaskVector2dBySpawnGroup(
     Masked Vector2D=&RandomRangeVector2D(Minimum=<1 1.5>, Maximum=<1.5 1.5>)
     Passthrough Vector2D=&Vector2DFromFloat(Value=&RandomRangeFloat(Minimum=2, Maximum=10))
     Spawn Group Mask=0
    )
    Sprite Size Mode=::NewEnumerator4
    UsePositionOffset=false
    Write Material Random=true
   )
   &ShapeLocation(
    MaskSpawnGroup=false
    Override Seed=false
    Sphere Radius=&MaskFloatBySpawnGroup(Masked Float=0, Passthrough Float=&RandomRangeFloat(Minimum=2, Maximum=5))
    UseOwnerScale=true
   )
   &AddVelocity(Velocity=&MaskVectorBySpawnGroup(
     Masked Vector=&RandomRangeVector(Minimum=<-250 -250 -250>, Maximum=<250 250 250>)
     Passthrough Vector=&Multiply_VectorByFloat(Vector=$Particles.ShapeLocation.ShapeVector, Float=5)
    ))
   &Color(Color=&MakeLinearColorFromVectorAndFloat(
     "Vector (RGB)"=&MaskVectorBySpawnGroup(
      Masked Vector=<10 5.3681 0.47>
      Passthrough Vector=&RandomRangeVector(Minimum=<5 1.0294 1.3281>, Maximum=<20 3.336 0>)
     )
     "Float (Alpha)"=&ClampFloat(Float=1)
    ))
   &InheritVelocity(Inherited Velocity Amount Scale=<0.2 0.2 0.2>, Inherited Velocity Speed Limit=0, Speed Limit=false, UseSpeedThreshold=false)
  ]
  particleUpdateScript [
   &ParticleState()
   &ScaleSpriteSize(
    Scale Factor=&MaskVector2dBySpawnGroup(
     Masked Vector2D=&Vector2DFromCurve(
      Vector2Curve=DataInterface() {propertyValues {
        XCurve {keys [<at 0 1 RCIM_Cubic>, {interpMode=RCIM_Cubic, leaveTangent=-1.9252, time=1, value=0.2}]}
        YCurve {keys [<at 0 3 RCIM_Cubic>,<at 1 0 RCIM_Cubic>]}
       }}
      ScaleCurve=<1 1>
     )
     Passthrough Vector2D=&Vector2DFromFloat(Value=&FloatFromCurve(FloatCurve=DataInterface() {propertyValues {Curve {keys [<at 0 0>,<at 1 1>]}}}))
     Spawn Group Mask=0
    )
    Scale Sprite Size Mode=::NewEnumerator2
   )
   &AccelerationForce(Acceleration=&MaskVectorBySpawnGroup(Masked Vector=<0 0 -980>, Passthrough Vector=<0 0 0>))
   &Drag(
    Drag=&MaskFloatBySpawnGroup(Masked Float=3, Passthrough Float=1)
    Use Linear Drag=true
    Use Rotational Drag=true
   )
   &InheritSourceMovement(
    Applied Position Scale=&VectorFromCurve(VectorCurve=DataInterface() {propertyValues {XCurve {keys [<at 0.5 1>,<at 1 0>]}, YCurve {keys [<at 0.5 1>,<at 1 0>]}, ZCurve {keys [<at 0.5 1>,<at 1 0>]}}})
    ApplyForce=false
    ApplyPosition=true
    ApplyPositionFalloff=false
   )
   &SolveForcesAndVelocity(Clamp Velocity=false, Limit Acceleration=false)
   &SubUVAnimation(Number Of Frames=&MaskFloatBySpawnGroup(Masked Float=0, Passthrough Float=3))
  ]
  rendererClasses [&NiagaraSpriteRendererProperties]
  renderers [
   {rendererClass=&NiagaraSpriteRendererProperties}
   {rendererClass=&NiagaraSpriteRendererProperties, rendererIndex=1}
  ]
 }
 Ribbon_Smoke {
  emitterSpawnScript [/SetVariables() {isSetParameters=true}]
  emitterUpdateScript [
   &EmitterState(Life Cycle Mode=::NewEnumerator1, Loop Delay=0.27, Loop Duration=1, UseLoopDelay=false)
   &SpawnPerUnit(
    Module.Use Max Movement Threshold=true
    Module.Use Movement Tolerance=true
    Movement Tolerance=0.1
    Spawn Spacing:NiagaraFloat=&Divide_Float(A=1, B=&Divide_Float(A=1, B=25))
    Use Spawn Probability=false
   )
  ]
  particleSpawnScript [
   /SetVariables(Particles.Acceleration:Vector3f=<0 0 250>) {isSetParameters=true}
   &InitializeRibbon(Lifetime=0.7, Position Offset:Vector3f=<0 0 12>, UsePositionOffset=false, Write Material Random=true)
   &DynamicMaterialParameters(Module.Index0Param0Write=true, Module.Index0Param1Write=false, Module.Index0Param2Write=false, Module.Index0Param3Write=false)
  ]
  particleUpdateScript [
   /SetVariables(Particles.ScaledRibbonWidth:NiagaraFloat=&Multiply_Float(
     A=$Particles.RibbonWidth
     B=&FloatFromCurve(FloatCurve=DataInterface() {propertyValues {Curve {keys [
          {arriveTangent=9.5068, interpMode=RCIM_Cubic, time=-0.0056, value=0.0871}
          {arriveTangent=0.0733, leaveTangent=0.0733, time=0.9928, value=6.8584}
         ]}}})
    )) {isSetParameters=true}
   &ParticleState()
   &Color(Color=&MakeLinearColorFromVectorAndFloat(
     "Vector (RGB)"=&VectorFromCurve(VectorCurve=DataInterface() {propertyValues {
        XCurve {keys [
          <at 0.1712 6.2357>
          {arriveTangent=-19.1968, interpMode=RCIM_Cubic, leaveTangent=-19.1968, time=0.3943, value=0.5958}
          {arriveTangent=-1.5995, interpMode=RCIM_Cubic, leaveTangent=-1.5995, time=0.4653, value=0.5907}
          <at 0.73 0.0589 RCIM_Cubic>
         ]}
        YCurve {keys [
          <at 0.1712 1.0363>
          {arriveTangent=-1.5241, interpMode=RCIM_Cubic, leaveTangent=-1.5241, time=0.3943, value=0.6055}
          {arriveTangent=-1.6299, interpMode=RCIM_Cubic, leaveTangent=-1.6299, time=0.4653, value=0.5881}
          <at 0.73 0.0583 RCIM_Cubic>
         ]}
        ZCurve {keys [
          <at 0.1712 0.4143>
          {arriveTangent=0.6727, interpMode=RCIM_Cubic, leaveTangent=0.6727, time=0.3943, value=0.6173}
          {arriveTangent=-1.6543, interpMode=RCIM_Cubic, leaveTangent=-1.6543, time=0.4653, value=0.6121}
          <at 0.73 0.062 RCIM_Cubic>
         ]}
       }})
     "Float (Alpha)"=&ClampFloat(Float=&FloatFromCurve(FloatCurve=DataInterface() {propertyValues {Curve {keys [<at 0.1016 1>, {arriveTangent=-1.1131, interpMode=RCIM_Cubic, leaveTangent=-1.1131, time=0.4637, value=0.2405}, <at 1 0>]}}}))
    ))
   &AccelerationForce(Acceleration=<0 0 30>)
   &SolveForcesAndVelocity(Clamp Velocity=false, Limit Acceleration=false)
  ]
  rendererClasses [&NiagaraRibbonRendererProperties]
  renderers [{rendererClass=&NiagaraRibbonRendererProperties}]
 }
}
summary {
 systemName=NS_TNT_Fuse_Converted
 emitters [
  {emitterName=FireGradient, rendererClasses [&NiagaraSpriteRendererProperties, &NiagaraLightRendererProperties]}
  {emitterName=LongLived_drips, simTarget=GPUComputeSim, rendererClasses [&NiagaraSpriteRendererProperties]}
  {emitterName=Ribbon_Smoke, rendererClasses [&NiagaraRibbonRendererProperties]}
 ]
 userVariables []
}
