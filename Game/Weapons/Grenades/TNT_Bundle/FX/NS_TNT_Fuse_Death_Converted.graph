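# Niagara system for the TNT bundle grenade's fuse death effect, a brief fiery glow with a flickering light.
# FireGradient001 partitions one spawn stream into large faint glow sprites and small flame sprites, switching size, lifetime, color, camera offset and SubUV by partition.
# Emitter-level waveform and color values supply the light radius and color alongside two sprite renderers.
# Five LW_ emitters show no renderers or modules here, and two older emitters are disabled.
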NiagaraSystem /Game/Weapons/Grenades/TNT_Bundle/FX/NS_TNT_Fuse_Death_Converted
dependencies=null
emitters {
 FireGradient001 {
  emitterUpdateScript [
   &EmitterState(Inactive Response=::NewEnumerator1, Life Cycle Mode=::NewEnumerator1, Loop Duration=1, UseLoopDelay=false)
   &SpawnRate(SpawnRate=50, Use Spawn Probability=false)
   /SetVariables(
    Emitter.LightColor:LinearColor=&ColorFromCurve(
     ColorCurve=DataInterface() {propertyValues {
       AlphaCurve {keys [<at 0 1 RCIM_Cubic>]}
       BlueCurve {keys [<at 0 0 RCIM_Cubic>]}
       GreenCurve {keys [<at 0 0.6 RCIM_Cubic>]}
       RedCurve {keys [<at 0 16 RCIM_Cubic>]}
      }}
     CurveIndex=$Emitter.NormalizedLoopAge
    )
    Emitter.LightExponent=1.5
    Emitter.LightRadius:NiagaraFloat=&Waveform(
     Waveform Count=::NewEnumerator1
     Global Amplitude Scale=165
     "Waveform [1]"=::NewEnumerator0
     "[1] Phase Scale"=5
     "[1] Frequency"=1
     "[1] Amplitude Min/Max"=<0 1>
     "[1] Bias"=0.2
     "[2] Average"=false
     "[2] Input Position"=0
     "[2] Phase"=$Emitter.Age
     "[2] Phase Scale"=1.1
     "[2] Compound Frequency"=2
     "[2] Frequency"=2
     "[2] Amplitude Min/Max"=<0 1>
     "[2] Scale"=1
     "[2] Bias"=0.5
     "[2] Exponent"=1
     "[2] Pulse Duration"=0.1
     "[2] Clip Triangle Min/Max"=<0 0>
     "[2] Random Seed"=0
    )
   ) {isSetParameters=true}
  ]
  particleSpawnScript [
   &PartitionParticles()
   /SetVariables(Particles.VisibilityTag=$Particles.PartitionParticles.Partition) {isSetParameters=true}
   &InitializeParticle(
    Lifetime=&MakeCustomFloatFromBool(
     Bool=&SetBoolByFloatComparison(A=0.5, B=&MakeFloatFromInt(Int=$Particles.PartitionParticles.Partition))
     True Float=&RandomRangeFloat(Minimum=0.1, Maximum=0.2)
     False Float=&RandomRangeFloat(Minimum=0.025, Maximum=0.1)
    )
    Position Offset=<0 0 15>
    Sprite Rotation Mode=::NewEnumerator2
    Sprite Size Mode=::NewEnumerator3
    Uniform Sprite Size=&MakeCustomFloatFromBool(
     Bool=&SetBoolByFloatComparison(A=0.5, B=&MakeFloatFromInt(Int=$Particles.PartitionParticles.Partition))
     True Float=75
     False Float=&RandomRangeFloat(Minimum=5, Maximum=20)
    )
    UsePositionOffset=false
    Write Material Random=true
   )
   &Color(Color=&MakeLinearColorFromVectorAndFloat(
     "Vector (RGB)"=&SelectVectorFromArray(
      Direct Array Index=$Particles.PartitionParticles.Partition
      Vector Selection Array=DataInterface() {propertyValues {FloatData [<5 5 5>, <5 1.3858 0.7867>]}}
     )
     "Float (Alpha)"=&SelectFloatFromArray(Direct Array Index=$Particles.PartitionParticles.Partition, Float Selection Array=DataInterface() {propertyValues {FloatData [0.1,1]}})
    ))
   &SystemLocation(MaskSpawnGroup=false, Offset=<0 0 16>)
  ]
  particleUpdateScript [
   &ParticleState()
   &ScaleSpriteSize(
    Scale Sprite Size Mode=::NewEnumerator0
    Uniform Scale Factor=&MakeCustomFloatFromBool(
     Bool=&SetBoolByFloatComparison(A=0.5, B=&MakeFloatFromInt(Int=$Particles.PartitionParticles.Partition))
     True Float=1
     False Float=&FloatFromCurve(FloatCurve=DataInterface() {propertyValues {Curve {keys [{tangentMode=RCTM_User, time=0, value=0}, {tangentMode=RCTM_User, time=1, value=1}]}}})
    )
   )
   &ScaleColor(
    Scale Alpha=&FloatFromCurve(FloatCurve=DataInterface() {propertyValues {Curve {keys [
         {interpMode=RCIM_Cubic, tangentMode=RCTM_User, time=0, value=0}
         {interpMode=RCIM_Cubic, tangentMode=RCTM_User, time=0.5, value=1}
         {tangentMode=RCTM_User, time=1, value=0}
        ]}}})
    Scale RGB=&MakeCustomVectorFromBool(
     Bool=&SetBoolByFloatComparison(A=0.5, B=&MakeFloatFromInt(Int=$Particles.PartitionParticles.Partition))
     True Vector=&VectorFromCurve(
      VectorCurve=DataInterface() {propertyValues {
        XCurve {keys [<at 0 0.11>,<at 0.5 1>,<at 1 0.11>]}
        YCurve {keys [<at 0 0.05>,<at 0.5 0.13>,<at 1 0.05>]}
        ZCurve {keys [
          {arriveTangent=-0.3559, interpMode=RCIM_Cubic, leaveTangent=-0.3559, tangentMode=RCTM_User, time=0, value=0.03}
          {arriveTangent=0.6186, interpMode=RCIM_Cubic, leaveTangent=-0.8742, tangentMode=RCTM_Break, time=0.5, value=0}
          <at 1 0.03>
         ]}
       }}
      Scale Curve=<2 2 2>
     )
     False Vector=<1 1 1>
    )
    ScaleA=true
    ScaleRGB=true
   )
   &DynamicMaterialParameters(
    Index 0 Param 1=&FloatFromCurve(FloatCurve=DataInterface() {propertyValues {Curve {keys [<at 0 8>,<at 0.5 5>,<at 1 8>]}}})
    Index 0 Param 2=1
    Index 0 Param 3=1
    Index 0 Param 4=1
    Module.Index0Param0Write=true
    Module.Index0Param1Write=true
    Module.Index0Param2Write=true
    Module.Index0Param3Write=true
   )
   &CameraOffset(Camera Offset Amount=&SelectFloatFromArray(Direct Array Index=$Particles.PartitionParticles.Partition, Float Selection Array=DataInterface() {propertyValues {FloatData [50,2]}}))
   /Niagara/Modules/Update/SubUV/SubUVAnimation(Number Of Frames=&SelectFloatFromArray(Direct Array Index=$Particles.PartitionParticles.Partition, Float Selection Array=DataInterface() {propertyValues {FloatData [0,3]}}))
  ]
  rendererClasses [&NiagaraSpriteRendererProperties, &NiagaraLightRendererProperties]
  renderers [
   {rendererClass=&NiagaraSpriteRendererProperties}
   {rendererClass=&NiagaraLightRendererProperties, rendererIndex=1}
   {rendererClass=&NiagaraSpriteRendererProperties, rendererIndex=2}
  ]
 }
 LW_InnerFlameA {rendererClasses [], renderers []}
 LW_InnerFlameA001 {rendererClasses [], renderers []}
 LW_InnerFlameA002 {rendererClasses [], renderers []}
 LW_LongLived_drips002 {rendererClasses [], renderers []}
 LW_Single_Fire {rendererClasses [], renderers []}
}
summary {
 systemName=NS_TNT_Fuse_Death_Converted
 emitters [
  {bEnabled=false, emitterName=InnerFlameA, rendererClasses [&NiagaraMeshRendererProperties]}
  {
   emitterName=FireGradient001
   rendererClasses [&NiagaraSpriteRendererProperties, &NiagaraLightRendererProperties]
  }
  {bEnabled=false, emitterName=LongLived_drips002, simTarget=GPUComputeSim, rendererClasses [&NiagaraSpriteRendererProperties]}
  {emitterName=LW_InnerFlameA, rendererClasses []}
  {emitterName=LW_InnerFlameA001, rendererClasses []}
  {emitterName=LW_InnerFlameA002, rendererClasses []}
  {emitterName=LW_LongLived_drips002, rendererClasses []}
  {emitterName=LW_Single_Fire, rendererClasses []}
 ]
 userVariables []
}
