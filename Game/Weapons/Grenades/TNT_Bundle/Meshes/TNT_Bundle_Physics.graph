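# Physics asset for the TNT bundle grenade's skeletal mesh, giving the bundle and its fuse simple collision and joints for physics.
# A large capsule on root covers the bundle, while two small capsules on dyn_fuse1 and dyn_fuse2 form the fuse.
# Constraints chain root to dyn_fuse1 to dyn_fuse2.
# Swing and twist are free on both joints, so their stored angle limits do not restrict the fuse.
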PhysicsAsset /Game/Weapons/Grenades/TNT_Bundle/Meshes/TNT_Bundle_Physics
bodies {
 dyn_fuse1 [{shapeType=Capsule, length=5.3373, radius=1.6551, center=<3.9884 -0.2117 -0.511>, rotation=<rot 0 91.2105 -88.9904>}]
 dyn_fuse2 [{shapeType=Capsule, length=5.6806, radius=1.4074, center=<5.1571 0.7686 0.1905>, rotation=<rot 0 -105.6727 -86.842>}]
 root [{shapeType=Capsule, length=16.3082, radius=5.4211, center=<-0.0049 -0.0004 1.032>, rotation=<rot -87.7906 4.3165 -4.3173>}]
}
constraints [
 {
  bone1Name=dyn_fuse1
  bone2Name=root
  swing1LimitDegrees=45
  swing1Motion=Free
  swing2LimitDegrees=45
  swing2Motion=Free
  twistLimitDegrees=45
  twistMotion=Free
 }
 {
  bone1Name=dyn_fuse2
  bone2Name=dyn_fuse1
  swing1LimitDegrees=45
  swing1Motion=Free
  swing2LimitDegrees=45
  swing2Motion=Free
  twistLimitDegrees=45
  twistMotion=Free
 }
]
